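# Physics asset for the M_MED_BirdNest_Navy male player body cosmetic, covering its dangling zipper and collar bones.
# Capsule bodies sit on the collar and zipper bones, with two larger capsules on spine_05 forming the upper-torso volume.
# The collar bones form a tightly limited chain with no constraint to the spine.
# The zipper base swings and twists freely on spine_05, while the zipper tip hangs from it with asymmetric swing limits.
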PhysicsAsset /BRCosmetics/Characters/Player/Male/Medium/Bodies/M_MED_BirdNest_Navy/Meshes/M_MED_BirdNest_Navy_Physics
bodies {
 dyn_collar_01 [{shapeType=Capsule, length=0.505, radius=0.505, shapeName=dyn_collar_01_capsule, rotation=<rot 0 -90 0>}]
 dyn_collar_03 [{shapeType=Capsule, length=8.25, radius=3.5, shapeName=dyn_collar_03_capsule, rotation=<rot 90 0 0>}]
 dyn_collar_04 [{shapeType=Capsule, length=8.25, radius=3.5, shapeName=dyn_collar_04_capsule, rotation=<rot 90 0 0>}]
 dyn_zipper [{
   shapeType=Capsule
   length=4.005
   radius=0.505
   shapeName=dyn_zipper_capsule
   center=<-2.0527 0.0219 -0.0175>
   rotation=<rot 0 89.8867 -89.8888>
  }]
 dyn_zipper_base1 [{shapeType=Capsule, length=0.505, radius=0.505, shapeName=dyn_zipper_base1_capsule, rotation=<rot 0 -90 0>}]
 spine_05 [
  {
   shapeType=Capsule
   length=8.3013
   radius=12.5019
   shapeName=spine_05_capsule
   center=<-1.177 0.7948 -0.2702>
   rotation=<rot 3.0476 -149.8487 -8.4113>
  }
  {shapeType=Capsule, length=11.75, radius=6.5, shapeName=spine_05_capsule, center=<-0.6573 5.4467 0>}
 ]
}
constraints [
 {bone1Name=dyn_collar_04, bone2Name=dyn_collar_03, swing1LimitDegrees=5, swing2LimitDegrees=5, twistLimitDegrees=0}
 {bone1Name=dyn_collar_03, bone2Name=dyn_collar_01, swing1LimitDegrees=2, swing2LimitDegrees=8, twistLimitDegrees=0}
 {
  bone1Name=dyn_zipper_base1
  bone2Name=spine_05
  swing1LimitDegrees=45
  swing1Motion=Free
  swing2LimitDegrees=45
  swing2Motion=Free
  twistLimitDegrees=45
  twistMotion=Free
 }
 {bone1Name=dyn_zipper, bone2Name=dyn_zipper_base1, swing1LimitDegrees=90, swing2LimitDegrees=10, twistLimitDegrees=0}
]
